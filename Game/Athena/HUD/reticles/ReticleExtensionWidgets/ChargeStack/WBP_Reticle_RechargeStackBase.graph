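# Base reticle extension widget that shows an equipped weapon's magazine ammo as recharge stack pips, with no widgets of its own.
# On construct it caches the weapon, generates pips from clip size and current ammo, and binds to local ammo changes, unbinding on destruct.
# Ammo changes update the stack count, while clip attribute changes regenerate the stacks.
# GetPipCollection returns nothing here, apparently left for child widgets to supply the pip collection.
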WidgetBlueprint /Game/Athena/HUD/reticles/ReticleExtensionWidgets/ChargeStack/WBP_Reticle_RechargeStackBase
graphs [
 BindToWeaponEvents {nodes [
   n1 = BindToWeaponEvents!() -> (then, DoBind:bool=~, EquippedWeapon:Fort Weapon Object Reference) {is=title K2Node_FunctionEntry}
   > Is Valid(InputObject=@n1.EquippedWeapon)
   > n2 = Branch(Condition=@n1)
   n3 = Create Event()
   n4 = Get EquippedWeapon() -> (EquippedWeapon:Fort Weapon Object Reference)
   > Bind Event to On Local Ammo Changed(self:Fort Weapon Object Reference=@n4, Delegate=@n3) {is=title K2Node_AddDelegate}
   Unbind Event from On Local Ammo Changed(execute=@n2.else, self:Fort Weapon Object Reference=@n4, Delegate=@n3) {is=title K2Node_RemoveDelegate}
  ]}
 EventGraph {nodes [
   n1 = OnLocalAmmoChanged!() -> (OutputDelegate, then, LocalCount:int, LocalRemaining:int) {is=title K2Node_CustomEvent}
   > UpdateStackCount!(
    execute
    self:WBP Reticle Recharge Stack Base Pip Collection Object Reference=GetPipCollection!(self) -> (PipCollection:WBP Reticle Recharge Stack Base Pip Collection Object Reference)
    NewCount:int=@n1.LocalCount
   ) -> (then)
   Event Construct()
   > n2 = Set CachedEquippedWeapon(CachedEquippedWeapon:Fort Weapon Object Reference=GetEquippedWeapon!(self:Fort Weapon Reticle Extension Widget Base Object Reference) -> (ReturnValue:Fort Weapon Object Reference)) -> (Output_Get:Fort Weapon Object Reference)
   > GenerateStacks!(
    execute
    self:WBP Reticle Recharge Stack Base Pip Collection Object Reference=GetPipCollection()
    TotalStacks:int=GetBulletsPerClip!(self:Fort Weapon Object Reference=@n2) -> (ReturnValue:int=0)
    CurrentStacks:int=GetLocalMagazineAmmoCount!(self:Fort Weapon Object Reference=@n2) -> (ReturnValue:int=0)
   ) -> (then)
   n3 = Get CachedEquippedWeapon() -> (CachedEquippedWeapon:Fort Weapon Object Reference)
   > BindToWeaponEvents!(execute, self, DoBind:bool=true, EquippedWeapon:Fort Weapon Object Reference=@n3) -> (then) {is=title K2Node_CallFunction}
   Event Destruct()
   > BindToWeaponEvents(DoBind=false, EquippedWeapon=@n3)
   Event OnAmmoClipAttributeChanged!() -> (OutputDelegate, then, AmmoClipValue:float) {is=title K2Node_Event}
   n4 = Get CachedEquippedWeapon()
   > Is Valid(InputObject=@n4)
   > GenerateStacks(
    self=GetPipCollection()
    TotalStacks=GetBulletsPerClip(self=@n4) -> (ReturnValue=0)
    CurrentStacks=GetLocalMagazineAmmoCount(self=@n4) -> (ReturnValue=0)
   )
  ]}
 GetPipCollection {nodes [
   GetPipCollection() {is=title K2Node_FunctionEntry}
   > Return Node!(execute, PipCollection:WBP Reticle Recharge Stack Base Pip Collection Object Reference)
  ]}
]
variables [
 CachedEquippedWeapon {type=FortWeapon, default=null}
 IsSpecialActive {type=boolean}
]
widgets {
 info {inheritedWidgetCount=0, parentClass=/Script/FortniteUI.FortWeaponReticleExtensionWidgetBase, rootWidgetClass=null, widgetCount=0}
 widgets []
}
