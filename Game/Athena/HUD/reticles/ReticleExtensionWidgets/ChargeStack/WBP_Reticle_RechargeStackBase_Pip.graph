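# Base widget Blueprint for one pip in a weapon reticle recharge-stack display, derived from FortWeaponReticleRechargeStackBasePipBase.
# It tracks an IsFilled flag, which InitFillState sets directly without animation.
# OnFill and OnUse change the flag and call PlayFillAnimation or PlayUseAnimation only when the state actually flips.
# Both animation functions are empty and the widget tree has no widgets, leaving visuals to be defined elsewhere.
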WidgetBlueprint /Game/Athena/HUD/reticles/ReticleExtensionWidgets/ChargeStack/WBP_Reticle_RechargeStackBase_Pip
graphs [
 EventGraph {links [], nodes []}
 InitFillState {nodes [
   n1 = InitFillState!() -> (then, IsFilled:bool=~) {is=title K2Node_FunctionEntry}
   > Set IsFilled(IsFilled:bool=@n1) -> (Output_Get:bool)
  ]}
 OnFill {nodes [
   OnFill() {is=title K2Node_FunctionEntry}
   > n1 = Branch(Condition=Get IsFilled() -> (IsFilled:bool))
   Set IsFilled(execute=@n1.else, IsFilled=true)
   > PlayFillAnimation()
  ]}
 OnUse {nodes [
   OnUse() {is=title K2Node_FunctionEntry}
   > Branch(Condition=Get IsFilled())
   > Set IsFilled(IsFilled=false)
   > PlayUseAnimation()
  ]}
 PlayFillAnimation {nodes [PlayFillAnimation() {is=title K2Node_FunctionEntry}]}
 PlayUseAnimation {nodes [PlayUseAnimation() {is=title K2Node_FunctionEntry}]}
]
variables [IsFilled {type=boolean}]
widgets {
 info {inheritedWidgetCount=0, parentClass=/Script/FortniteUI.FortWeaponReticleRechargeStackBasePipBase, rootWidgetClass=null, widgetCount=0}
 widgets []
}
