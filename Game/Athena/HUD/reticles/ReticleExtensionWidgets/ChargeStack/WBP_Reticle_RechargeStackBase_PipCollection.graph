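# Base widget for a weapon reticle's row of recharge-stack pips, with no widget tree of its own.
# GenerateStacks rebuilds the parent class's dynamic entry box with one pip per stack and sets each initial fill, optionally in reverse order.
# UpdateStackCount calls fill or use on every pip, plays an optional sound when leaving zero, and broadcasts old and new counts.
# Empty OnStackCreated and OnStackCountUpdated functions act as hooks.
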WidgetBlueprint /Game/Athena/HUD/reticles/ReticleExtensionWidgets/ChargeStack/WBP_Reticle_RechargeStackBase_PipCollection
graphs [
 EventGraph {links [], nodes []}
 GenerateStacks {nodes [
   n1 = GenerateStacks!() -> (then, TotalStacks:int, CurrentStacks:int) {is=title K2Node_FunctionEntry}
   > Set CurrentCount(CurrentCount:int=@n1.CurrentStacks) -> (Output_Get:int=0)
   > Reset!(
    execute
    self:Dynamic Entry Box Object Reference=GetDynamicEntryBox!(self:Fort Weapon Reticle Recharge Stack Base Pip Collection Base Object Reference) -> (ReturnValue:Dynamic Entry Box Object Reference)
    bDeleteWidgets:bool=false
   ) -> (then)
   > n2 = For Loop(FirstIndex=1, LastIndex=@n1)
   > n3 = Create Entry!(execute, self:Dynamic Entry Box Object Reference=GetDynamicEntryBox()) -> (then, ReturnValue:User Widget Object Reference)
   > n4 = Cast To WBP_Reticle_RechargeStackBase_Pip!(execute, Object=@n3) -> (then, CastFailed, AsWBP Reticle Recharge Stack Base Pip:WBP Reticle Recharge Stack Base Pip Object Reference) {is=title K2Node_DynamicCast}
   > OnStackCreated!(execute, self, StackWidget:WBP Reticle Recharge Stack Base Pip Object Reference=@n4) -> (then)
   n5 = Get CurrentCount() -> (CurrentCount:int=0)
   n6 = int - int(
    A=GetNumEntries!(self:Dynamic Entry Box Base Object Reference=GetDynamicEntryBox()) -> (ReturnValue:int=0)
    B=Get CurrentCount() -> (CurrentCount=0)
   )
   > InitFillState!(
    execute
    self:WBP Reticle Recharge Stack Base Pip Object Reference=@n4
    IsFilled:bool=Select(
     Option 0:bool="integer <= integer"(A=@n2, B=@n5)
     Option 1:bool="integer > integer"(A=@n2, B=@n6)
     Index=Get FillInReverseOrder() -> (FillInReverseOrder:bool)
    ) -> (ReturnValue:bool)
   ) -> (then)
  ]}
 GetTotalCount {nodes [
   GetTotalCount() {is=title K2Node_FunctionEntry}
   n1 = GetDynamicEntryBox()
   > n2 = Is Valid(InputObject=@n1)
   Return Node!(execute=@n2.Is Not Valid, TotalCount:int=0)
   > Return Node(TotalCount=Length(TargetArray:Array of User Widget Object References=GetAllEntries!(self:Dynamic Entry Box Base Object Reference=@n1) -> (ReturnValue:Array of User Widget Object References)))
  ]}
 OnStackCountUpdated {nodes [OnStackCountUpdated!() -> (then, OldCount:int, NewCount:int) {is=title K2Node_FunctionEntry}]}
 OnStackCreated {nodes [OnStackCreated!() -> (then, StackWidget:WBP Reticle Recharge Stack Base Pip Object Reference) {is=title K2Node_FunctionEntry}]}
 PlayAmmoChangedSounds {nodes [
   n1 = PlayAmmoChangedSounds!() -> (then, NewCount:int) {is=title K2Node_FunctionEntry}
   > Is Valid(InputObject=Get RechargeFirstChargeSound() -> (RechargeFirstChargeSound:Sound Cue Object Reference))
   n2 = Get CurrentCount() -> (CurrentCount=0)
   > Branch(Condition=AND Boolean(A="Equal (Integer)"(A=@n2), B="integer > integer"(A=@n1)))
   > PlaySound2D(Sound=Get RechargeFirstChargeSound())
  ]}
 StackCountUpdatedEvent {nodes [StackCountUpdatedEvent!() -> (then, OldCount:int, NewCount:int) {is=title K2Node_FunctionEntry}]}
 UpdateStackCount {nodes [
   n1 = UpdateStackCount!() -> (then, NewCount:int) {is=title K2Node_FunctionEntry}
   > Branch(Condition="Not Equal (Integer)"(A=@n1, B=Get CurrentCount() -> (CurrentCount=0)))
   > PlayAmmoChangedSounds!(execute, self, NewCount:int=@n1) -> (then) {is=title K2Node_CallFunction}
   > n2 = For Each Loop(Array:Array of User Widget Object References=GetAllEntries(self=GetDynamicEntryBox())) -> (Array Element:User Widget Object Reference)
   Set PrevCount(execute=@n2.Completed, PrevCount:int=Get CurrentCount() -> (CurrentCount=0)) -> (Output_Get:int=0)
   n3 = Get NewCount() -> (NewCount:int=0)
   > Set CurrentCount(CurrentCount=@n3) -> (Output_Get=0)
   > OnStackCountUpdated!(execute, self, OldCount:int=Get PrevCount() -> (PrevCount:int=0), NewCount:int=@n3) -> (then) {is=title K2Node_CallFunction}
   > Call Stack Count Updated Event!(
    execute:Exec
    self:WBP Reticle Recharge Stack Base Pip Collection Object Reference
    OldCount:int=Get PrevCount() -> (PrevCount=0)
    NewCount:int=@n3
   ) -> (then:Exec) {is=title K2Node_CallDelegate}
   n4 = Cast To WBP_Reticle_RechargeStackBase_Pip(execute=@n2.LoopBody, Object=@n2)
   n5 = Get NewCount() -> (NewCount=0)
   n6 = int - int(A=GetNumEntries(self=GetDynamicEntryBox()) -> (ReturnValue=0), B=Get NewCount() -> (NewCount=0))
   > n7 = Branch(Condition=Select(
     Option 0="integer < integer"(A=@n2.Array Index, B=@n5)
     Option 1="integer >= integer"(A=@n2.Array Index, B=@n6)
     Index=Get FillInReverseOrder()
    ))
   > OnFill(self:WBP Reticle Recharge Stack Base Pip Object Reference=@n4)
   OnUse(execute=@n7.else, self:WBP Reticle Recharge Stack Base Pip Object Reference=@n4)
  ]}
]
variables [
 FillInReverseOrder {type=boolean}
 RechargeFirstChargeSound {type=SoundCue, default=null, class=~, properties=~}
 StackCountUpdatedEvent {}
]
widgets {
 info {
  inheritedWidgetCount=0
  parentClass=/Script/FortniteUI.FortWeaponReticleRechargeStackBasePipCollectionBase
  rootWidgetClass=null
  widgetCount=0
 }
 widgets []
}
